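; Blueprint for the Rumble Male pickaxe, whose fruit pieces disappear progressively as it lands hits.
; Impacts advance a hit counter that steps the fruit selection at set thresholds, hiding mesh elements through a dynamic material and triggering bone-attached Niagara destruction.
; The selection also recolors and reattaches swing FX, tints impact FX and switches impact sound, with the final stage using default alpha.
; Equipping resets everything, while frontend logic loops the counter.
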(event Custom|ResetColorSwing
  (bind _returnvalue (CallFunction|GetSwingVFXComponent))
  ;[1 1 1] Fruit Swing Colors and Knife Alpha setting
  (Effects|Components|ParticleSystem|SetVectorParameter _returnvalue "FruitColor" (Math|Conversions|ToVector(LinearColor) (Utilities|Array|Get(aref) (Variables|Default|GetFruitColor) (Variables|Default|GetFruitSelection))))
  (Effects|Components|ParticleSystem|SetFloatParameter _returnvalue "Alpha" 1.0)
  (Effects|Components|ParticleSystem|SetFloatParameter _returnvalue "AlphaDefault" 0.0))

(event Custom|FruitColoring
  (bind _returnvalue (CallFunction|GetSwingVFXComponent))
  (bind _fruitselection (Variables|Default|GetFruitSelection))
  ;[1 1 1] Fruit Swing Colors and Knife Alpha setting
  (if (== _fruitselection 4)
    (Effects|Components|ParticleSystem|SetFloatParameter _returnvalue "Alpha" 0.0)
    (Effects|Components|ParticleSystem|SetFloatParameter _returnvalue "AlphaDefault" 1.0)
    (else
      (Effects|Components|ParticleSystem|SetVectorParameter _returnvalue "FruitColor" (Math|Conversions|ToVector(LinearColor) (Utilities|Array|Get(aref) (Variables|Default|GetFruitColor) _fruitselection)))
      (Effects|Components|ParticleSystem|SetFloatParameter _returnvalue "Alpha" 1.0)
      (Effects|Components|ParticleSystem|SetFloatParameter _returnvalue "AlphaDefault" 0.0))))

(event EventTick
  (params (DeltaSeconds :type Float%20%28single-precision%29)))

(event Custom|ImpactHit
  (bind _pickaxe_impact (Variables|Default|GetPickaxe%20Impact))
  (bind (_bblockinghit _binitialoverlap _time _distance _location) (Collision|BreakHitResult _pickaxe_impact))
  (bind _impact_ref (Variables|Default|GetImpact_Ref))
  (bind _fruitselection (Variables|Default|GetFruitSelection))
  ;[1 1 1] Spawn an Impact FX at location
  (bind _returnvalue (Niagara|SpawnSystematLocation (Variables|Default|GetImpactFX) _location))
  (Variables|Default|SetImpact_Ref _returnvalue)
  (CallFunction|UpdateImpactSound _location)
  (Utilities|FlowControl|DoOnce false
    (:Completed
      (if (== _fruitselection 4)
        (Effects|Components|ParticleSystem|SetFloatParameter _impact_ref "Alpha" 0.0)
        (Effects|Components|ParticleSystem|SetFloatParameter _impact_ref "AlphaDefault" 1.0)
        (else
          (Effects|Components|ParticleSystem|SetVectorParameter _impact_ref "FruitColor" (Math|Conversions|ToVector(LinearColor) (Utilities|Array|Get(aref) (Variables|Default|GetFruitColor) _fruitselection)))))
      (CallFunction|ResetImpact))))

(event FortWeapon|EventOnPlayImpactFX
  (params (HitResult :type Hit%20Result%20Structure%20%28by%20ref%29) (ImpactPhysicalSurface :type EPhysicalSurface%20Enum :default "SurfaceType_Default") (SpawnedPSC :type FXSystem%20Component%20Object%20Reference))
  (bind _hitcounter (Variables|Default|GetHitCounter))
  ;[1 1 1] Impact FX
  (Variables|Default|SetPIckaxe%20Impact HitResult)
  (|CallParentFunction HitResult ImpactPhysicalSurface SpawnedPSC)
  (Utilities|Casting|CastToFortGameStateFrontEnd (Game|GetGameState)
    (:then
      ;; Handle Frontend Cool Stuff But Slightly Tidier_2
      (CallFunction|HandleFrontedHotnessLogic)
      (Utilities|SetImpactHitResultbyComponent (Variables|Visual|GetWeaponMesh)))
    (:CastFailed
      _))
  ;;[0.4125 0.4818 1] Primary - Inform Items Hiding after number of Hits
  (Utilities|FlowControl|Sequence
    (:then_0
      (Variables|Default|SetHitCounter (+ (Variables|Default|GetHitCounter) 1))
      (CallFunction|ImpactHit))
    (:then_1
      (if (== _hitcounter 3)
        (Variables|Default|SetFruitSelection 1)))
    (:then_2
      (if (== _hitcounter 5)
        (Variables|Default|SetFruitSelection 2)))
    (:then_3
      (if (== _hitcounter 7)
        (Variables|Default|SetFruitSelection 3)))
    (:then_4
      (if (== _hitcounter 9)
        (Variables|Default|SetFruitSelection 4))))
  (bind _returnvalue (Fortnite|GetDynamicMaterialInstance (Variables|Visual|GetWeaponMesh)))
  (Rendering|Material|SetScalarParameterValue _returnvalue (Utilities|Array|Get(aref) (Variables|Default|GetHideElements) (Variables|Default|GetFruitSelection)) 1.0 :!class /Script/Engine.MaterialInstanceDynamic)
  (CallFunction|FruitDestruction)
  (CallFunction|ResetAttachment))

(event Custom|ResetImpact
  (bind _impact_ref (Variables|Default|GetImpact_Ref))
  (bind _fruitselection (Variables|Default|GetFruitSelection))
  ;[1 1 1] Spawn an Impact FX at location
  (Utilities|FlowControl|DoOnce false
    (:Completed
      (if (== _fruitselection 4)
        (Effects|Components|ParticleSystem|SetFloatParameter _impact_ref "Alpha" 0.0)
        (Effects|Components|ParticleSystem|SetFloatParameter _impact_ref "AlphaDefault" 1.0)
        (else
          (Effects|Components|ParticleSystem|SetVectorParameter _impact_ref "FruitColor" (Math|Conversions|ToVector(LinearColor) (Utilities|Array|Get(aref) (Variables|Default|GetFruitColor) _fruitselection)))))
      (CallFunction|ResetImpact))))

(event FortWeapon|EventOnEquippedAndAttached
  ;[1 1 1] Reset on Equip
  (|CallParentFunction)
  (Utilities|FlowControl|Sequence
    (:then_0
      (Variables|Default|SetHitCounter 0)
      (Variables|Default|SetFruitSelection 0))
    (:then_1
      (bind _returnvalue (Fortnite|GetDynamicMaterialInstance (Variables|Visual|GetWeaponMesh)))
      (Rendering|Material|SetScalarParameterValue _returnvalue "Hide Element 01" 0.0 :!class /Script/Engine.MaterialInstanceDynamic)
      (Rendering|Material|SetScalarParameterValue _returnvalue "Hide Element 02" 0.0 :!class /Script/Engine.MaterialInstanceDynamic)
      (Rendering|Material|SetScalarParameterValue _returnvalue "Hide Element 03" 0.0 :!class /Script/Engine.MaterialInstanceDynamic)
      (Rendering|Material|SetScalarParameterValue _returnvalue "Hide Element 04" 0.0 :!class /Script/Engine.MaterialInstanceDynamic))
    (:then_2
      (CallFunction|ResetAttachment)
      (CallFunction|ResetColorSwing))))

(event Custom|ResetFruitLogic
  ;[1 1 1] Reset on Equip
  (Utilities|FlowControl|Sequence
    (:then_0
      (Variables|Default|SetHitCounter 0)
      (Variables|Default|SetFruitSelection 0))
    (:then_1
      (bind _returnvalue (Fortnite|GetDynamicMaterialInstance (Variables|Visual|GetWeaponMesh)))
      (Rendering|Material|SetScalarParameterValue _returnvalue "Hide Element 01" 0.0 :!class /Script/Engine.MaterialInstanceDynamic)
      (Rendering|Material|SetScalarParameterValue _returnvalue "Hide Element 02" 0.0 :!class /Script/Engine.MaterialInstanceDynamic)
      (Rendering|Material|SetScalarParameterValue _returnvalue "Hide Element 03" 0.0 :!class /Script/Engine.MaterialInstanceDynamic)
      (Rendering|Material|SetScalarParameterValue _returnvalue "Hide Element 04" 0.0 :!class /Script/Engine.MaterialInstanceDynamic))
    (:then_2
      (CallFunction|ResetAttachment)
      (CallFunction|ResetColorSwing))))

(event Custom|FruitDestruction
  (bind _fx_niagara_destruction (Variables|Default|GetFX_Niagara_Destruction))
  ;[1 1 1] Attach Fruit Destruction to proper bone and execute destruction FX.
  (Utilities|FlowControl|DoOnce false
    (:Completed
      (Transformation|AttachComponentToComponent (Variables|Default|GetFX_Niagara_Destruction) (Variables|Visual|GetWeaponMesh) (Utilities|Array|Get(acopy) (Variables|Default|GetBoneAttachment) (Variables|Default|GetFruitSelection)) "KeepRelative" "KeepRelative" "KeepRelative" true)
      (Effects|Components|ParticleSystem|SetVectorParameter (Variables|Default|GetFX_Niagara_Destruction) "FruitColor" (Math|Conversions|ToVector(LinearColor) (Utilities|Array|Get(aref) (Variables|Default|GetFruitDestructionColor) (Variables|Default|GetFruitSelection))))
      (Components|Activation|Activate _fx_niagara_destruction false)
      (Utilities|FlowControl|Delay 0.25)
      (Components|Activation|Deactivate _fx_niagara_destruction)
      (CallFunction|ResetDestruction))))

(event Custom|ResetDestruction
  (bind _fx_niagara_destruction (Variables|Default|GetFX_Niagara_Destruction))
  ;[1 1 1] Attach Fruit Destruction to proper bone and execute destruction FX.
  (Utilities|FlowControl|DoOnce false
    (:Completed
      (Transformation|AttachComponentToComponent (Variables|Default|GetFX_Niagara_Destruction) (Variables|Visual|GetWeaponMesh) (Utilities|Array|Get(acopy) (Variables|Default|GetBoneAttachment) (Variables|Default|GetFruitSelection)) "KeepRelative" "KeepRelative" "KeepRelative" true)
      (Effects|Components|ParticleSystem|SetVectorParameter (Variables|Default|GetFX_Niagara_Destruction) "FruitColor" (Math|Conversions|ToVector(LinearColor) (Utilities|Array|Get(aref) (Variables|Default|GetFruitDestructionColor) (Variables|Default|GetFruitSelection))))
      (Components|Activation|Activate _fx_niagara_destruction false)
      (Utilities|FlowControl|Delay 0.25)
      (Components|Activation|Deactivate _fx_niagara_destruction)
      (CallFunction|ResetDestruction))))

(event Custom|ResetAttachment
  ;[1 1 1] Set Swing FX to proper bone selection
  (Utilities|FlowControl|DoOnce false
    (:Completed
      (CallFunction|FruitColoring)
      (Transformation|AttachComponentToComponent (CallFunction|GetSwingVFXComponent) (Variables|Visual|GetWeaponMesh) (Utilities|Array|Get(acopy) (Variables|Default|GetBoneAttachment) (+ (Variables|Default|GetFruitSelection) 1)) "KeepRelative" "KeepRelative" "KeepRelative" true))))

(event EventBeginPlay
  ;[1 1 1] DoOnce to prevent Destruction to happen on spawn.
  (|CallParentFunction)
  (Utilities|FlowControl|Sequence
    (:then_0
      (Utilities|FlowControl|DoOnce false
        (:Completed
          (Utilities|FlowControl|Delay 0.5)
          (Variables|Default|SetDoOnce true))))
    (:then_1
      (Transformation|AttachComponentToComponent (Variables|Default|GetFX_Niagara_Destruction) (Variables|Visual|GetWeaponMesh) "dyn_watermelon" "KeepRelative" "KeepRelative" "KeepRelative" true))))

(event EventOnPlayWeaponFireFX
  (params (bPersistentFire :type Boolean) (bSecondaryFire :type Boolean))
  ;[1 1 1] Set Swing FX to proper bone selection
  (|CallParentFunction bPersistentFire bSecondaryFire)
  (Utilities|FlowControl|DoOnce false
    (:Completed
      (CallFunction|FruitColoring)
      (Transformation|AttachComponentToComponent (CallFunction|GetSwingVFXComponent) (Variables|Visual|GetWeaponMesh) (Utilities|Array|Get(acopy) (Variables|Default|GetBoneAttachment) (+ (Variables|Default|GetFruitSelection) 1)) "KeepRelative" "KeepRelative" "KeepRelative" true))))

(fn HandleFrontedHotnessLogic
  (if (== (Variables|Default|GetHitCounter) 0)
    (Variables|Default|SetHitCounter 2)
    (elif (> (Variables|Default|GetHitCounter) 9)
      (CallFunction|ResetFruitLogic) ; But we have to unhide the fruit!
      (Variables|Default|SetHitCounter 1))))

(fn UpdateImpactSound
  (params (Location :type Vector))
  (bind _returnvalue (Audio|SpawnSoundatLocation (Variables|Default|GetImpactSound) Location))
  (Audio|Parameter|SetBooleanParameter _returnvalue "FruitSelection" (!= (Variables|Default|GetFruitSelection) 4) :!class /Script/AudioExtensions.AudioParameterControllerInterface))

(fn ConstructionScript
  (|CallParentFunction))
