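# Niagara destruction effect for the Rumble pickaxe, bursting fruit-like debris and juice tinted by User.FruitColor with random brightness per particle.
# Meshes and Sprites emit shrinking chunks via a burst plus brief decaying spawn rate, with an emitter-age curve weakening launch velocity under gravity, drag and curl noise.
# Splash adds rising, expanding sprites, and Splotch a brief fixed-facing flash, both animated through dynamic material parameters.
# User.DripSpawnRate is declared but unused.
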NiagaraSystem /BRCosmetics/Weapons/FORT_Melee/Pickaxe_Rumble_Male/FX/NS_Rumble_Pickaxe_Destruction
dependencies {usedDataInterfaces [&NiagaraDataInterfaceCurve], usedDynamicInputs [], usedModules [&SystemState]}
emitters {
 Meshes {
  emitterUpdateScript [
   &EmitterState()
   &SpawnRate(
    SpawnRate=&FloatFromCurve(
     FloatCurve=DataInterface() {propertyValues {Curve {keys [
         {arriveTangent=0.1596, interpMode=RCIM_Cubic, leaveTangent=0.1596, tangentMode=RCTM_User, time=0, value=1}
         {arriveTangent=-1.9409, interpMode=RCIM_Cubic, leaveTangent=-1.9409, tangentMode=RCTM_User, time=0.3, value=0}
        ]}}}
     CurveIndex=$Emitter.Age
     Scale Curve=10
    )
    Use Spawn Probability=false
   )
   &SpawnBurst_Instantaneous(Spawn Count=4, Use Spawn Probability=false, UseLoopCountLimit=false)
   /SetVariables(Emitter.AgeCurve:NiagaraFloat=&FloatFromCurve(
     FloatCurve=DataInterface() {propertyValues {Curve {keys [
         {arriveTangent=-0.6647, interpMode=RCIM_Cubic, leaveTangent=-0.6647, tangentMode=RCTM_User, time=0, value=1}
         {arriveTangent=-3.3333, interpMode=RCIM_Cubic, leaveTangent=-3.3333, time=0.3, value=0}
        ]}}}
     CurveIndex=$Emitter.Age
    )) {isSetParameters=true}
  ]
  particleSpawnScript [
   &InitializeParticle(
    Color=&MakeLinearColorFromVectorAndFloat(
     "Vector (RGB)"=&Multiply_Vector(A=$User.FruitColor, B=&RandomRangeVector(Minimum=<0.75 0.75 0.75>, Maximum=<1 1 1>))
     "Float (Alpha)"=1
    )
    Lifetime Max=0.6
    Lifetime Min=0.3
    Lifetime Mode=::NewEnumerator1
    Mesh Scale Mode=::NewEnumerator3
    Mesh Uniform Scale=&FloatFromCurve(
     FloatCurve=DataInterface() {propertyValues {Curve {keys [<at 0 0.06>,<at 0.15 0.055>,<at 0.3 0.03>,<at 1 0.02>]}}}
     CurveIndex=&RandomRangeFloat(Minimum=0, Maximum=1)
    )
    UsePositionOffset=false
    Write Material Random=true
   )
   &InitialMeshOrientation(Use Rotation Vector=false)
   &AddVelocity(Velocity=&Multiply_VectorByFloat(Vector=&RandomRangeVector(Minimum=<0 50 0>, Maximum=<0 100 0>), Float=$Emitter.AgeCurve))
   &AddVelocityFromPoint(
    Module.Use Falloff Distance=false
    Velocity Strength=&Multiply_Float(A=&RandomRangeFloat(Minimum=100, Maximum=200), B=$Emitter.AgeCurve)
   )
   &SubUVAnimation(End Frame=3, SubUV Animation Mode:ENiagara_SubUVLookupMode=::NewEnumerator2)
  ]
  particleUpdateScript [
   &Drag(Use Linear Drag=true, Use Rotational Drag=false)
   &GravityForce(Gravity=<0 0 -700>)
   &CurlNoiseForce(Noise Frequency=100, Noise Strength=50, Pan Noise=false)
   &SolveForcesAndVelocity(Clamp Velocity=false, Limit Acceleration=false)
   &ParticleState()
   &ScaleMeshSize(Scale Factor=&VectorFromFloat(Value=&FloatFromCurve(FloatCurve=DataInterface() {propertyValues {Curve {keys [
          {interpMode=RCIM_Cubic, tangentMode=RCTM_User, time=0, value=1}
          {arriveTangent=-1.4104, interpMode=RCIM_Cubic, leaveTangent=-1.4104, tangentMode=RCTM_User, time=1, value=0}
         ]}}})))
   &UpdateMeshOrientation(
    Rotation Rate=&RandomRangeFloat(Minimum=1, Maximum=3)
    Rotation Vector=&RandomVector(Vector Scale=1)
   )
  ]
  rendererClasses [&NiagaraMeshRendererProperties]
  renderers [{rendererClass=&NiagaraMeshRendererProperties, rendererIndex=0}]
 }
 Splash {
  emitterUpdateScript [
   &EmitterState()
   &SpawnBurst_Instantaneous(Spawn Count=3, Use Spawn Probability=false)
  ]
  particleSpawnScript [
   &InitializeParticle(
    Color=&MakeLinearColorFromVectorAndFloat(
     "Vector (RGB)"=&Multiply_Vector(A=$User.FruitColor, B=&RandomRangeVector(Minimum=<0.9 0.9 0.9>, Maximum=<1.1 1.1 1.1>))
     "Float (Alpha)"=2
    )
    Lifetime=&RandomRangeFloat(Minimum=0.375, Maximum=0.5)
    Sprite Rotation Mode=::NewEnumerator2
    Sprite Size Mode=::NewEnumerator3
    Uniform Sprite Size=&RandomRangeFloat(Minimum=0.5, Maximum=1)
    UsePositionOffset=false
    Write Material Random=true
   )
   &SphereLocation(
    MaskSpawnGroup=false
    Offset=<3 -4 -7>
    Override Seed=false
    Sphere Radius=2
    Surface Only=false
    UseNonUniformScale=false
    UseOffset=true
    UseSphereOrientation=false
   )
   &CameraOffset(Camera Offset Amount=&RandomRangeFloat(Minimum=0, Maximum=1))
   &AddVelocity(Rotation Coordinate Space=::World, Velocity=&RandomRangeVector(Minimum=<0 0 50>, Maximum=<0 0 100>))
  ]
  particleUpdateScript [
   &DynamicMaterialParameters(
    Index 0 Param 1=&FloatFromCurve(FloatCurve=DataInterface() {propertyValues {Curve {keys [
         {arriveTangent=6.3575, interpMode=RCIM_Cubic, leaveTangent=6.3575, tangentMode=RCTM_User, time=0, value=1}
         {arriveTangent=1.4386, interpMode=RCIM_Cubic, leaveTangent=1.4386, tangentMode=RCTM_User, time=1, value=4}
        ]}}})
    Index 0 Param 2=&FloatFromCurve(FloatCurve=DataInterface() {propertyValues {Curve {keys [
         {arriveTangent=2.4412, interpMode=RCIM_Cubic, leaveTangent=2.4412, tangentMode=RCTM_User, time=0, value=0}
         {arriveTangent=0.1511, interpMode=RCIM_Cubic, leaveTangent=0.1511, tangentMode=RCTM_User, time=1, value=1}
        ]}}})
    Index 0 Param 4=2.9
    Module.Param0WriteEnabled=true
    Module.Param1WriteEnabled=true
    Module.Param2WriteEnabled=true
    Module.Param3WriteEnabled=true
   )
   &InheritVelocity(Module.Limit Speed=true, Velocity Limit=0.5)
   &GravityForce(Gravity=<0 0 -500>)
   &SolveForcesAndVelocity(Clamp Velocity=false, Limit Acceleration=false)
   &ParticleState()
   &ScaleSpriteSize(Scale Factor=&Vector2DFromFloat(Value=&FloatFromCurve(FloatCurve=DataInterface() {propertyValues {Curve {keys [
          {arriveTangent=48.4699, interpMode=RCIM_Cubic, leaveTangent=48.4699, tangentMode=RCTM_User, time=0, value=20}
          {
           arriveTangent=33.8415
           arriveTangentWeight=0.0333
           interpMode=RCIM_Cubic
           leaveTangent=33.8415
           leaveTangentWeight=0.3
           tangentMode=RCTM_User
           time=0.4
           value=37
          }
          {arriveTangent=-0.5994, arriveTangentWeight=0.3333, interpMode=RCIM_Cubic, leaveTangent=-0.5994, tangentMode=RCTM_User, time=1, value=44}
         ]}}})))
  ]
  rendererClasses [&NiagaraSpriteRendererProperties]
  renderers [{rendererClass=&NiagaraSpriteRendererProperties}]
 }
 Splotch {
  emitterUpdateScript [
   &EmitterState()
   &SpawnBurst_Instantaneous(Use Spawn Probability=false, UseLoopCountLimit=false)
  ]
  particleSpawnScript [
   &InitializeParticle(
    Color=&MakeLinearColorFromVectorAndFloat(
     "Vector (RGB)"=&Multiply_Vector(A=$User.FruitColor, B=&RandomRangeVector(Minimum=<0.9 0.9 0.9>, Maximum=<1.1 1.1 1.1>))
     "Float (Alpha)"=1
    )
    Lifetime=0.13
    Sprite Rotation Angle=&RandomRangeFloat(Minimum=0, Maximum=1)
    Sprite Rotation Mode=::NewEnumerator1
    Sprite Size=<15 15>
    Sprite Size Mode=::NewEnumerator4
    UsePositionOffset=false
    Write Material Random=true
   )
   &CameraOffset(Camera Offset Amount=10)
   /SetVariables(Particles.SpriteFacing=<0 1 0>) {isSetParameters=true}
   &SkeletalMeshLocation(Bone Sampling Mode=::NewEnumerator4, MaskSpawnGroup=false, Sampled Position Offset=<0 -10 -10>, Skeletal Mesh=$User.Skeletal Mesh) {enabled=false}
  ]
  particleUpdateScript [
   &ParticleState()
   &ScaleColor(
    Scale Alpha=&FloatFromCurve(FloatCurve=DataInterface() {propertyValues {Curve {keys [
         {arriveTangent=28.077, interpMode=RCIM_Cubic, leaveTangent=28.077, tangentMode=RCTM_User, time=0, value=0}
         {arriveTangent=0.3546, interpMode=RCIM_Cubic, leaveTangent=0.3546, tangentMode=RCTM_User, time=0.1408, value=3}
         {arriveTangent=-3.0042, interpMode=RCIM_Cubic, leaveTangent=-3.0042, tangentMode=RCTM_User, time=1, value=0}
        ]}}})
    ScaleA=true
    ScaleRGB=true
   )
   &DynamicMaterialParameters(
    Index 0 Param 1=0.1
    Index 0 Param 2=3
    Index 0 Param 3=&FloatFromCurve(FloatCurve=DataInterface() {propertyValues {Curve {keys [<at 0 0.5>,<at 1 1>]}}})
    Index 0 Param 4=1
    Module.Param0WriteEnabled=true
    Module.Param1WriteEnabled=true
    Module.Param2WriteEnabled=true
    Module.Param3WriteEnabled=true
   )
   &ScaleSpriteSize(Uniform Curve Sprite Scale=DataInterface() {propertyValues {Curve {keys [
        {arriveTangent=-0.1154, interpMode=RCIM_Cubic, leaveTangent=-0.1154, tangentMode=RCTM_User, time=0, value=0}
        {arriveTangent=0.5008, interpMode=RCIM_Cubic, leaveTangent=0.5008, tangentMode=RCTM_User, time=1, value=1}
       ]}}})
  ]
  rendererClasses [&NiagaraSpriteRendererProperties]
  renderers [{rendererClass=&NiagaraSpriteRendererProperties}]
 }
 Sprites {
  emitterUpdateScript [
   &EmitterState()
   &SpawnRate(
    SpawnRate=&FloatFromCurve(
     FloatCurve=DataInterface() {propertyValues {Curve {keys [
         {arriveTangent=0.1596, interpMode=RCIM_Cubic, leaveTangent=0.1596, tangentMode=RCTM_User, time=0, value=1}
         {arriveTangent=-1.9409, interpMode=RCIM_Cubic, leaveTangent=-1.9409, tangentMode=RCTM_User, time=0.3, value=0}
        ]}}}
     CurveIndex=$Emitter.Age
     Scale Curve=10
    )
    Use Spawn Probability=false
   )
   &SpawnBurst_Instantaneous(Spawn Count=5, Use Spawn Probability=false, UseLoopCountLimit=false)
   /SetVariables(Emitter.AgeCurve:NiagaraFloat=&FloatFromCurve(
     FloatCurve=DataInterface() {propertyValues {Curve {keys [
         {arriveTangent=-0.6647, interpMode=RCIM_Cubic, leaveTangent=-0.6647, tangentMode=RCTM_User, time=0, value=1}
         {arriveTangent=-3.3333, interpMode=RCIM_Cubic, leaveTangent=-3.3333, time=0.3, value=0}
        ]}}}
     CurveIndex=$Emitter.Age
    )) {isSetParameters=true}
  ]
  particleSpawnScript [
   &InitializeParticle(
    Color=&MakeLinearColorFromVectorAndFloat(
     "Vector (RGB)"=&Multiply_Vector(A=$User.FruitColor, B=&RandomRangeVector(Minimum=<0.75 0.75 0.75>, Maximum=<1 1 1>))
     "Float (Alpha)"=1
    )
    Lifetime Max=0.6
    Lifetime Min=0.3
    Lifetime Mode=::NewEnumerator1
    Sprite Rotation Mode=::NewEnumerator2
    Sprite Size Mode=::NewEnumerator3
    Uniform Sprite Size=&FloatFromCurve(
     FloatCurve=DataInterface() {propertyValues {Curve {keys [<at 0 6>,<at 0.2 5.25>,<at 0.3 3.5>,<at 1 2>]}}}
     CurveIndex=&RandomRangeFloat(Minimum=0, Maximum=1)
    )
    UsePositionOffset=false
    Write Material Random=true
   )
   &AddVelocity(Velocity=&Multiply_VectorByFloat(Vector=&RandomRangeVector(Minimum=<0 100 0>, Maximum=<0 150 0>), Float=$Emitter.AgeCurve))
   &AddVelocityFromPoint(
    Module.Use Falloff Distance=false
    Velocity Strength=&Multiply_Float(A=&RandomRangeFloat(Minimum=100, Maximum=200), B=$Emitter.AgeCurve)
   )
   &SubUVAnimation(End Frame=3, SubUV Animation Mode:ENiagara_SubUVLookupMode=::NewEnumerator2)
  ]
  particleUpdateScript [
   &Drag(Drag=5, Use Linear Drag=true, Use Rotational Drag=false)
   &GravityForce(Gravity=<0 0 -700>)
   &CurlNoiseForce(Noise Frequency=100, Noise Strength=50, Pan Noise=false)
   &SolveForcesAndVelocity(Clamp Velocity=false, Limit Acceleration=false)
   &ParticleState()
   &ScaleSpriteSize(Uniform Curve Sprite Scale=DataInterface() {propertyValues {Curve {keys [
        {interpMode=RCIM_Cubic, tangentMode=RCTM_User, time=0, value=1}
        {arriveTangent=-2.5513, interpMode=RCIM_Cubic, leaveTangent=-2.5513, tangentMode=RCTM_User, time=1, value=0}
       ]}}})
  ]
  rendererClasses [&NiagaraSpriteRendererProperties]
  renderers [{rendererClass=&NiagaraSpriteRendererProperties}]
 }
}
summary {
 systemName=NS_Rumble_Pickaxe_Destruction
 emitters [
  {emitterName=Splash, rendererClasses [&NiagaraSpriteRendererProperties]}
  {emitterName=Splotch, rendererClasses [&NiagaraSpriteRendererProperties]}
  {emitterName=Sprites, rendererClasses [&NiagaraSpriteRendererProperties]}
  {emitterName=Meshes, rendererClasses [&NiagaraMeshRendererProperties]}
 ]
 userVariables [
  User.DripSpawnRate {type=NiagaraFloat, defaultValue=0, description=""}
  User.FruitColor {type=Vector3f, description="", defaultValue {struct=Vector3f, value=<0 0 0>}}
 ]
}
