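# Physics asset for the Rumble Male two-handed pickaxe, defining collision bodies and joint limits for its dangling, dynamic parts.
# Four capsule joints each constrain directly to the two-capsule handle, swinging mainly on one axis with tight or locked secondary swing.
# A pineapple greenery capsule hangs off the fourth joint.
# Left and right ties are four-box chains from the handle, each link limited in swing and twist against its parent.
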PhysicsAsset /BRCosmetics/Weapons/FORT_Melee/Pickaxe_Rumble_Male/Meshes/Rumble_Male_Axe_PhysicsAsset
bodies {
 Melee_TwoHanded_Handle [
  {shapeType=Capsule, length=135.1691, radius=2.343, center=<1.011 -0.9321 49.474>, rotation=<rot -0.0275 -90.0489 -0.2982>}
  {
   shapeType=Capsule
   length=15.75
   radius=3
   shapeName=Melee_TwoHanded_Handle_capsule
   center=<0.7338 -0.502 -18.363>
   rotation=<rot 10.2906 -0.4559 -4.4282>
  }
 ]
 dyn_joint1 [{shapeType=Capsule, length=17.25, radius=7.75, shapeName=dyn_joint1_capsule, center=<0 -0.0138 -0.1304>, rotation=<rot 0 0 -83.963>}]
 dyn_joint2 [{shapeType=Capsule, length=7.75, radius=4.75, shapeName=dyn_joint2_capsule, center=<0 -1.078 -0.993>, rotation=<rot 0 0 -84.0494>}]
 dyn_joint3 [{shapeType=Capsule, length=4.75, radius=7.25, shapeName=dyn_joint3_capsule, center=<-1.2167 -0.2775 0.3303>, rotation=<rot 0 0 -88.007>}]
 dyn_joint4 [{shapeType=Capsule, length=15.75, radius=14.25, shapeName=dyn_joint4_capsule, center=<-0.3492 0.3584 -0.9127>, rotation=<rot 0 0 -80.9112>}]
 dyn_l_tie_joint [{
   shapeType=Box
   extentX=0.5
   extentY=2.135
   extentZ=2.1875
   shapeName=dyn_l_tie_joint_box
   center=<-0.0531 0.0753 -0.0412>
   rotation=<rot -1.1774 -90.0044 -1.0453>
  }]
 dyn_l_tie_joint1 [{
   shapeType=Box
   extentX=0.5
   extentY=1.885
   extentZ=2.5725
   shapeName=dyn_l_tie_joint1_box
   center=<0.049 0.0252 -0.0817>
   rotation=<rot -4.5087 -90.0001 0.9394>
  }]
 dyn_l_tie_joint2 [{
   shapeType=Box
   extentX=0.5625
   extentY=1.635
   extentZ=2.6975
   shapeName=dyn_l_tie_joint2_box
   center=<0.1334 -0.0228 -0.0297>
   rotation=<rot -6.17 -90.0002 -0.7096>
  }]
 dyn_l_tie_joint3 [{
   shapeType=Box
   extentX=0.6875
   extentY=2.4576
   extentZ=3.4236
   shapeName=dyn_l_tie_joint3_box
   center=<0.4069 -0.0821 0.0917>
   rotation=<rot -8.9807 -91.0645 -3.5522>
  }]
 dyn_pineapple_greenery_joint [{
   shapeType=Capsule
   length=7.5
   radius=8.5
   shapeName=dyn_pineapple_greenery_joint_capsule
   center=<0 -12.4503 -0.7458>
   rotation=<rot 0 0 -96.0449>
  }]
 dyn_r_tie_joint [{
   shapeType=Box
   extentX=0.5
   extentY=1.26
   extentZ=1.8225
   shapeName=dyn_r_tie_joint_box
   center=<0.2174 0.024 0.0062>
   rotation=<rot -20.8211 -90.1368 -3.1571>
  }]
 dyn_r_tie_joint1 [{
   shapeType=Box
   extentX=0.5
   extentY=1.51
   extentZ=2.26
   shapeName=dyn_r_tie_joint1_box
   center=<0.0974 -0.0082 0.0057>
   rotation=<rot -12.8739 -90.0005 0.9179>
  }]
 dyn_r_tie_joint2 [{
   shapeType=Box
   extentX=0.5
   extentY=1.76
   extentZ=2.885
   shapeName=dyn_r_tie_joint2_box
   center=<0.0247 -0.0403 0.1242>
   rotation=<rot -3.6718 -90.059 10.8519>
  }]
 dyn_r_tie_joint3 [{
   shapeType=Box
   extentX=0.5967
   extentY=1.9938
   extentZ=3.4312
   shapeName=dyn_r_tie_joint3_box
   center=<0.5658 -0.0574 -0.0082>
   rotation=<rot 2.7821 -88.9996 0.6658>
  }]
}
constraints [
 {bone1Name=dyn_joint1, bone2Name=Melee_TwoHanded_Handle, swing1LimitDegrees=30, swing2LimitDegrees=7, twistLimitDegrees=10}
 {
  bone1Name=dyn_joint2
  bone2Name=Melee_TwoHanded_Handle
  swing1LimitDegrees=30
  swing2LimitDegrees=1
  swing2Motion=Locked
  twistLimitDegrees=5
 }
 {
  bone1Name=dyn_joint3
  bone2Name=Melee_TwoHanded_Handle
  swing1LimitDegrees=20
  swing2LimitDegrees=1
  swing2Motion=Locked
  twistLimitDegrees=8
 }
 {bone1Name=dyn_joint4, bone2Name=Melee_TwoHanded_Handle, swing1LimitDegrees=50, swing2LimitDegrees=3, twistLimitDegrees=0.1}
 {bone1Name=dyn_pineapple_greenery_joint, bone2Name=dyn_joint4, swing1LimitDegrees=20, swing2LimitDegrees=20, twistLimitDegrees=20}
 {bone1Name=dyn_l_tie_joint, bone2Name=Melee_TwoHanded_Handle, swing1LimitDegrees=20, swing2LimitDegrees=20, twistLimitDegrees=15}
 {bone1Name=dyn_l_tie_joint1, bone2Name=dyn_l_tie_joint, swing1LimitDegrees=20, swing2LimitDegrees=20, twistLimitDegrees=15}
 {bone1Name=dyn_l_tie_joint2, bone2Name=dyn_l_tie_joint1, swing1LimitDegrees=20, swing2LimitDegrees=20, twistLimitDegrees=15}
 {bone1Name=dyn_l_tie_joint3, bone2Name=dyn_l_tie_joint2, swing1LimitDegrees=20, swing2LimitDegrees=20, twistLimitDegrees=15}
 {bone1Name=dyn_r_tie_joint, bone2Name=Melee_TwoHanded_Handle, swing1LimitDegrees=20, swing2LimitDegrees=20, twistLimitDegrees=15}
 {bone1Name=dyn_r_tie_joint1, bone2Name=dyn_r_tie_joint, swing1LimitDegrees=20, swing2LimitDegrees=20, twistLimitDegrees=15}
 {bone1Name=dyn_r_tie_joint2, bone2Name=dyn_r_tie_joint1, swing1LimitDegrees=20, swing2LimitDegrees=20, twistLimitDegrees=15}
 {bone1Name=dyn_r_tie_joint3, bone2Name=dyn_r_tie_joint2, swing1LimitDegrees=20, swing2LimitDegrees=20, twistLimitDegrees=15}
]
